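; Parent Blueprint for bouncy props that launch players, AI, vehicles and eligible projectiles on overlapping Jump_Collision.
; Overlap triggers a wobble, then a capsule ceiling trace picks the launch height for pawns and vehicles.
; Launched pawns get a gameplay event and effect plus fall-damage immunity, with cooldown-limited bounce sound.
; Construction applies palette color swatches, optionally seeded by world location, and shows debug collision in editor.
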(event Custom|MakeWobbleEffect
  (params (Center :type Vector))
  (CallFunction|StartBounceAnimation (Utilities|Time|GetTimeSeconds) 2e+02 (Math|Vector|MakeVector 0.0 0.0 1.0) Center))

(event OnComponentBeginOverlap(Jump_Collision)
  (params (OverlappedComponent :type Primitive%20Component%20Object%20Reference) (OtherActor :type Actor%20Object%20Reference) (OtherComp :type Primitive%20Component%20Object%20Reference) (OtherBodyIndex :type Integer) (bFromSweep :type Boolean) (SweepResult :type Hit%20Result%20Structure%20%28by%20ref%29))
  (bind _returnvalue (Actor|GetComponentbyClass OtherActor "/Script/Engine.ProjectileMovementComponent"))
  (bind (_bblockinghit _binitialoverlap _time _distance _location _impactpoint) (Collision|BreakHitResult SweepResult))
  (bind (_asfort_projectile_petrol _bsuccess) (Utilities|Casting|CastToFortProjectilePetrol OtherActor))
  (bind (_asfort_pickup _bsuccess_2) (Utilities|Casting|CastToFortPickup OtherActor))
  (bind _overlappawn (Variables|Default|GetOverlapPawn))
  (bind _overlappawn_2 (Variables|Default|GetOverlapPawn))
  (bind _overlappawn_3 (Variables|Default|GetOverlapPawn))
  (Variables|Default|SetBounceLocation (Transformation|GetActorLocation))
  (CallFunction|MakeWobbleEffect _impactpoint)
  (bind _asfort_player_pawn (Utilities|Casting|CastToFortPlayerPawn OtherActor)
    (:then
      (Variables|Default|SetOverlapPawn _asfort_player_pawn))
    (:CastFailed
      (bind _asfort_aipawn (Utilities|Casting|CastToFortAIPawn OtherActor)
        (:then
          (Variables|Default|SetOverlapPawn _asfort_aipawn))
        (:CastFailed
          (Utilities|FlowControl|Sequence
            (:then_0
              ;[1 0.03743 0] Launching Projectile
              (if (Math|Boolean|ANDBoolean (or (Utilities|IsValid _returnvalue :!class /Script/Engine.KismetSystemLibrary) (Utilities|IsValid (Actor|GetComponentbyClass OtherActor "/Script/FortniteGame.FortProjectileMovementComponent") :!class /Script/Engine.KismetSystemLibrary)) (not _bsuccess) (not _bsuccess_2) (Variables|Default|GetBBouncingProjectile))
                (bind _returnvalue_2 (Utilities|GetNonPawnBounceVel OtherActor)
                  (:Output
                    (bind (_returnvalue_x _returnvalue_y _returnvalue_z) (Math|Vector|BreakVector _returnvalue_2))
                    (Variables|/Script/Engine.MovementComponent|SetVelocity (Math|Vector|MakeVector _returnvalue_x _returnvalue_y _returnvalue_z) _returnvalue)))))
            (:then_1
              ;[1 1 1] Ceiling check
              (bind _success (CallFunction|CeilingCheck))
              (if _success
                (Variables|Default|SetBounceHeight 600)
                (Variables|Default|SetJumpHeight 100)
                (else
                  (Variables|Default|SetBounceHeight 2e+03)
                  (Variables|Default|SetJumpHeight 600)))
              (Variables|Default|SetFinalBounceJumpHeight (Variables|Default|GetBounceHeight))
              (bind _success_2 (Vehicles|LaunchHoverboard OtherActor (Math|Vector|MakeVector 0.0 0.0 (Variables|Default|GetFinalBounceJumpHeight)) "TireLaunch"))
              (if _success_2
                (Networking|ForceNetUpdate)
                (CallFunction|PlayBounceSound))))))))
  (Utilities|IsValid _overlappawn_3
    (:Is%20Valid
      (if (or (GameplayTags|HasMatchingGameplayTag (Ability|GetAbilitySystemComponent _overlappawn_3) "(TagName=\"Gameplay.Action.Player.Launched\")") (Variables|/Script/FortniteGame.FortPawn|GetbIsDBNO _overlappawn_3))
        _
        (else
          _)))
    (:Is%20Not%20Valid
      _))
  (bind _asfort_player_controller (Utilities|Casting|CastToFortPlayerController (Pawn|GetController _overlappawn_3))
    (:then
      (Variables|Default|SetOverlapController _asfort_player_controller))
    (:CastFailed
      _))
  ;[1 1 1] Ceiling check
  (bind _success_3 (CallFunction|CeilingCheck))
  (if _success_3
    (Variables|Default|SetBounceHeight 600)
    (Variables|Default|SetJumpHeight 100)
    (else
      (Variables|Default|SetBounceHeight (Variables|Default|GetNoCeilingBounceHeight))
      (Variables|Default|SetJumpHeight 600)))
  (Variables|Default|SetFinalBounceJumpHeight (Variables|Default|GetBounceHeight))
  (bind _asfort_player_pawn_2 (Utilities|Casting|CastToFortPlayerPawn (Variables|Default|GetOverlapPawn))
    (:then
      (RemoteControlPawn|ForceDetachfromRemoteControlledPawn _asfort_player_pawn_2))
    (:CastFailed
      _))
  (bind (_returnvalue_x_2 _returnvalue_y_2) (Math|Vector|BreakVector (Transformation|GetActorForwardVector _asfort_player_pawn_2)))
  (Character|LaunchCharacter _overlappawn_2 (Math|Vector|MakeVector (* _returnvalue_x_2 (Variables|Default|GetBounceXYVelocityMofifier)) (* _returnvalue_y_2 (Variables|Default|GetBounceXYVelocityMofifier)) (Variables|Default|GetFinalBounceJumpHeight)) false true)
  (Ability|SendGameplayEventtoActor _overlappawn_2 "(TagName=\"Event.Tire.Bounce\")")
  (GameplayEffects|ApplyGameplayEffectToSelf :self (Ability|GetAbilitySystemComponent _overlappawn) :GameplayEffectClass (Variables|Default|GetGE_PlayerLaunched) :Level 0.0)
  (Variables|/Script/FortniteGame.FortPawn|SetbIgnoreNextFallingDamage true _overlappawn)
  (Utilities|FlowControl|Sequence
    (:then_0
      (bind _overlapcontroller (Variables|Default|GetOverlapController)
        (:then
          (Game|Feedback|LocalPlayForceFeedback _overlapcontroller (Variables|Default|GetLaunchForceFeedback) false "TireLaunch"))
        (:else
          _))
      (Variables|Default|SetOverlapPawn 0)
      (Variables|Default|SetOverlapController 0))
    (:then_1
      (Networking|ForceNetUpdate)
      (CallFunction|PlayBounceSound))))

(event Custom|CustomEvent
  (params (OverlappedComponent :type Primitive%20Component%20Object%20Reference) (OtherActor :type Actor%20Object%20Reference) (OtherComp :type Primitive%20Component%20Object%20Reference) (OtherBodyIndex :type Integer) (bFromSweep :type Boolean) (SweepResult :type Hit%20Result%20Structure%20%28by%20ref%29)))

(event Custom|Play%20Bounce%20Sound
  ;[1 1 1] Audio - Play bounce sound
  (Networking|SwitchHasAuthority
    (:Authority
      (CallFunction|PlaySound))
    (:Remote
      (if (Variables|Default|GetCan%20Play%20Bounce%20Sound)
        (Audio|PlaySoundatLocation (Variables|Default|GetBounceSound) (Transformation|GetActorLocation))
        (Utilities|Time|SetTimerbyEvent (delegate Custom|Reset%20Bounce%20Sound%20Cooldown) 0.2)
        (Variables|Default|SetCan%20Play%20Bounce%20Sound false)))))

(event Custom|Play%20Sound
  ;[1 1 1] Audio - Play bounce sound
  (Networking|CanExecuteCosmeticEvents
    (:True
      (if (Variables|Default|GetCan%20Play%20Bounce%20Sound)
        (Audio|PlaySoundatLocation (Variables|Default|GetBounceSound) (Transformation|GetActorLocation))
        (Utilities|Time|SetTimerbyEvent (delegate Custom|Reset%20Bounce%20Sound%20Cooldown) 0.2)
        (Variables|Default|SetCan%20Play%20Bounce%20Sound false)))))

(event Custom|Reset%20Bounce%20Sound%20Cooldown
  ;[1 1 1] Audio - Play bounce sound
  (Variables|Default|SetCan%20Play%20Bounce%20Sound true))

(fn BuildExposedAssetProviderVersion(Implementation)
  (params (Version :type Fort%20Exposed%20Asset%20Provider%20Version%20Structure%20%28by%20ref%29))
  (|CallParentFunction Version)
  (ExposedAssetProvider|AddVersion Version 1)
  (return))

(fn CeilingCheck
  (bind _overlappawn (Variables|Default|GetOverlapPawn)
    (:then
      (Utilities|Array|Add (LocalVariables|GetActorsToIgnore) _overlappawn)
      (Utilities|Array|AppendArray (LocalVariables|GetActorsToIgnore) (Actor|GetAttachedActors _overlappawn true false))
      (for _array_element (Actor|GetAttachedActors _overlappawn true false)
        (Utilities|Array|AppendArray (LocalVariables|GetActorsToIgnore) (Actor|GetAttachedActors _array_element true false))))
    (:else
      _))
  (bind (_outhit _returnvalue) (Collision|CapsuleTraceByChannel (Math|Vector|MakeVector (.x (Transformation|GetWorldLocation (Variables|Transformation|GetRootComponent))) (.y (Transformation|GetWorldLocation (Variables|Transformation|GetRootComponent))) (+ (.z (Transformation|GetWorldLocation (Variables|Transformation|GetRootComponent))) 200)) (Math|Vector|MakeVector (.x (Transformation|GetWorldLocation (Variables|Transformation|GetRootComponent))) (.y (Transformation|GetWorldLocation (Variables|Transformation|GetRootComponent))) (+ (.z (Transformation|GetWorldLocation (Variables|Transformation|GetRootComponent))) 340)) 50 100 "TraceTypeQuery9" false (LocalVariables|GetActorsToIgnore)))
  (return _returnvalue))

(fn CheckJumpingOrientation
  (bind (_returnvalue_x _returnvalue_y) (Math|Vector|BreakVector (Transformation|GetActorForwardVector)))
  (bind _orientationmultiplier (Variables|Default|GetOrientationMultiplier))
  (if (== (Math|Float|Round (Math|Float|Absolute(Float) _returnvalue_x)) 1)
    (return (Math|Vector|vector*float (Math|Vector|MakeVector _returnvalue_x _returnvalue_y) _orientationmultiplier))
    (else
      (return (Math|Vector|vector*float (Math|Vector|MakeVector _returnvalue_x _returnvalue_y) (* _orientationmultiplier -1.0))))))

(fn CollectAssetReferencestoAllow(Implementation)
  (params (Collector :type Fort%20Exposed%20Asset%20Reference%20Collector%20Structure%20%28by%20ref%29))
  (|CallParentFunction Collector)
  (for _array_element 0
    (ExposedAssetProvider|CollectIndirectAssetReference Collector (Utilities|ResolveSoftReference _array_element)))
  (return))

(fn DebugCollisionSettings
  (bind _l_debugcomponent (LocalVariables|GetL_DebugComponent))
  (if (Variables|RandomColor|GetViewDebugCollision)
    (bind _spawnedcomponent (Fortnite|Utilities|TransientDuplicateSceneComponent (Variables|RandomColor|GetJump_Collision) self (Variables|Debug|GetStaticMeshComponent)))
    (Utilities|IsValid _spawnedcomponent
      (:Is%20Valid
        (bind _output_get (LocalVariables|SetL_DebugComponent _spawnedcomponent))
        (Utilities|IsValid _output_get
          (:Is%20Valid
            (Rendering|Material|SetMaterial (LocalVariables|GetL_DebugComponent) 0 (LocalVariables|GetL_DebugMaterial))
            (Rendering|SetVisibility (LocalVariables|GetL_DebugComponent) true false)
            (Variables|/Script/Engine.ActorComponent|SetComponentTags (LocalVariables|GetL_AR_Empty) (LocalVariables|GetL_DebugComponent))))))
    (else
      (Utilities|IsValid _l_debugcomponent
        (:Is%20Valid
          (Components|DestroyComponent _l_debugcomponent))))))

(fn RandomizeIndexByPosition
  (bind _ar_td_swatches (Variables|Default|GetAR_TD_Swatches))
  (bind (_outitem _outindex) (Utilities|Array|RandomArrayItemfromStream _ar_td_swatches (Math|Random|MakeRandomStreamfromLocation (Transformation|GetActorLocation) (Variables|RandomColor|GetTransitionDistance) true)))
  (CallFunction|ValidateSwatchPalette)
  (if (Variables|Default|GetConstructStatus)
    (if (> (Utilities|Array|Length _ar_td_swatches) 0)
      (Variables|RandomColor|SetSwatchIndex _outindex)
      (return))))

(fn UpdateDebugComponentSettings
  (params (Component :type Static%20Mesh%20Component%20Object%20Reference))
  (bind _output_get (LocalVariables|SetL_DebugComponent Component))
  (Utilities|IsValid _output_get
    (:Is%20Valid
      (Transformation|SetRelativeTransform (LocalVariables|GetL_DebugComponent) (Transformation|GetRelativeTransform (Variables|RandomColor|GetJump_Collision)) false false))))

(fn UpdateJumpCollisionVolume
  (LocalVariables|SetL_MeshCollision (Variables|RandomColor|GetJump_Collision))
  (switch Utilities|FlowControl|Switch|SwitchonENUM_RCPBounceProp_CollisionSelect (Variables|RandomColor|GetENUM_CollisionShape)
    (:NewEnumerator2
      (AddComponent|NODE_AddBoxComponent-1))
    (:NewEnumerator3
      (Game|AddComponentbyClass :Class (Utilities|GetClass (LocalVariables|GetL_MeshCollision))))))

(fn ConstructionScript
  (if (Editor|InEditororPIE)
    (CallFunction|DebugCollisionSettings))
  (if (Variables|RandomColor|GetRandomizeByWorldPosition)
    (CallFunction|RandomizeIndexbyPosition)
    (bind _buildingtexturedata (CallFunction|ValidateSwatchIndex))
    (Utilities|IsValid _buildingtexturedata
      (:Is%20Valid
        (CallFunction|SetBuildingTextureData 0 _buildingtexturedata)))
    (else
      (CallFunction|ValidateSwatchPalette)
      (bind _buildingtexturedata_2 (CallFunction|ValidateSwatchIndex))
      (Utilities|IsValid _buildingtexturedata_2
        (:Is%20Valid
          (CallFunction|SetBuildingTextureData 0 _buildingtexturedata_2))))))

(fn ValidateSwatchIndex
  (bind _ar_td_swatches (Variables|Default|GetAR_TD_Swatches))
  (if (Variables|Default|GetConstructStatus)
    (if (<= (Variables|RandomColor|GetSwatchIndex) (Variables|Default|GetMaxIndex))
      (return (Utilities|Array|Get(acopy) _ar_td_swatches (Variables|RandomColor|GetSwatchIndex)))
      (else
        (bind _output_get (Variables|RandomColor|SetSwatchIndex (Variables|Default|GetMaxIndex)))
        (return (Utilities|Array|Get(acopy) _ar_td_swatches _output_get))))))

(fn ValidateSwatchPalette
  (if (Utilities|IsValidClass (Variables|RandomColor|GetSwatchDataAsset))
    (bind _returnvalue (Game|ConstructObjectfromClass (Variables|RandomColor|GetSwatchDataAsset) self))
    (if (Utilities|IsValid _returnvalue :!class /Script/Engine.KismetSystemLibrary)
      (Variables|Default|SetConstructStatus true)
      (bind _output_get (Variables|RandomColor|SetColorPalette _returnvalue))
      (bind _output_get_2 (Variables|Default|SetAR_TD_Swatches (Variables|/Script/FortniteGame.FortRandomColorPropPalette|GetSwatch_Palette _output_get)))
      (Variables|Default|SetMaxIndex (- (Utilities|Array|Length _output_get_2) 1))
      (else
        (Variables|Default|SetConstructStatus false)))
    (return)))
